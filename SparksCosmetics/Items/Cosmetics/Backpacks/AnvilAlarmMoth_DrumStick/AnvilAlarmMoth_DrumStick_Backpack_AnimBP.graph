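# Animation Blueprint for the AnvilAlarmMoth drumstick backpack cosmetic, adding physics-driven sway to two dangling bone chains.
# AnimGraph offsets the offset, handle and chain-root bones, then runs per-bone AnimDynamics on three-link left and right chains, tuned differently per side.
# EventGraph derives backpack rotation and translation from spine bend and picks gravity by skydiving and parachute state.
# Those computed variables have no visible connection to the AnimGraph, which shows constant values.
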AnimBlueprint /SparksCosmetics/Items/Cosmetics/Backpacks/AnvilAlarmMoth_DrumStick/AnvilAlarmMoth_DrumStick_Backpack_AnimBP
target_skeleton=/SparksCosmetics/Instrument/DrumStick/AnvilAlarmMoth/Meshes/AnvilAlarmMoth_DrumStick_Backpack_Skeleton
graphs [
 AnimGraph {nodes [
   n1 = Local To Component(LocalPose=Input Pose())
   n2 = "Transform (Modify) Bone - Bone: dyn_chain_01_l"(
    ComponentPose="Transform (Modify) Bone - Bone: handle"(
     ComponentPose="Transform (Modify) Bone - Bone: offset"(ComponentPose=@n1, Alpha=1, Translation="0,-2,10", Rotation="0,0,0", Scale="1,1,1") {translationMode=BMM_Additive, boneToModify {boneName=offset}, translation {y=-2, z=10}}
     Alpha=1
     Translation="0,0,10"
     Rotation="0,0,0"
     Scale="1,1,1"
    ) {translationMode=BMM_Additive, boneToModify {boneName=handle}, translation {z=10}}
    Alpha=1
    Translation="0.1,-0.11,0"
    Rotation="0,0,0"
    Scale="1,1,1"
   ) {translationMode=BMM_Additive, translationSpace=BCS_BoneSpace, boneToModify {boneName=dyn_chain_01_l}, translation=<0.1 -0.11>}
   n3 = "Transform (Modify) Bone - Bone: dyn_chain_01_r"(ComponentPose=@n2, Alpha=1, Translation="0.1,0.2,0", Rotation="0,0,0", Scale="1,1,1") {translationMode=BMM_Additive, translationSpace=BCS_BoneSpace, boneToModify {boneName=dyn_chain_01_r}, translation=<0.1 0.2>}
   n4 = Anim Dynamics - Bone: dyn_chain_01_l!(ComponentPose=@n3, bAlphaBoolEnabled:bool=True, GravityOverride:Vector="0,0,-980", ExternalForce="0,0,0") -> (Pose) {
    alphaInputType=Bool
    angularDampingOverride=0.99
    bOverrideAngularDamping=true
    bUseGravityOverride=true
    linearDampingOverride=0.9
    alphaBoolBlend {blendOption=Linear}
    boundBone {boneName=dyn_chain_01_l}
    componentAppliedLinearAccClamp=<200 200 200>
    componentLinearAccScale=<2 2 2>
    componentLinearVelScale=<2 2 2>
    physicsBodyDefinitions [{
      boundBone {boneName=dyn_chain_01_l}
      boxExtents=<10 10 10>
      constraintSetup {angularLimitsMax=<0 0 80>, angularLimitsMin=<0 0 -80>, angularTarget=<0 0 0>}
      localJointOffset=<1 0 0>
     }]
   }
   n5 = Anim Dynamics - Bone: dyn_chain_02_l!(ComponentPose=@n4, bAlphaBoolEnabled:bool=True, GravityOverride:Vector="0,0,-980", ExternalForce="0,0,0") -> (Pose) {
    alphaInputType=Bool
    angularDampingOverride=0.99
    bOverrideAngularDamping=true
    bUseGravityOverride=true
    linearDampingOverride=0.9
    alphaBoolBlend {blendOption=Linear}
    boundBone {boneName=dyn_chain_02_l}
    componentAppliedLinearAccClamp=<200 200 200>
    componentLinearAccScale=<2 2 2>
    componentLinearVelScale=<2 2 2>
    physicsBodyDefinitions [{
      boundBone {boneName=dyn_chain_02_l}
      boxExtents=<10 10 10>
      constraintSetup {angularLimitsMax=<40 30 30>, angularLimitsMin=<-30 -30 -30>, angularTarget=<0 0 0>}
      localJointOffset=<1 0 0>
     }]
   }
   n6 = Anim Dynamics - Bone: dyn_chain_03_l!(ComponentPose=@n5, bAlphaBoolEnabled:bool=True, GravityOverride:Vector="0,0,-980", ExternalForce="0,0,0") -> (Pose) {
    alphaInputType=Bool
    angularDampingOverride=0.99
    bOverrideAngularDamping=true
    bUseGravityOverride=true
    linearDampingOverride=0.9
    alphaBoolBlend {blendOption=Linear}
    boundBone {boneName=dyn_chain_03_l}
    componentAppliedLinearAccClamp=<200 200 200>
    componentLinearAccScale=<2 2 2>
    componentLinearVelScale=<2 2 2>
    physicsBodyDefinitions [{
      boundBone {boneName=dyn_chain_03_l}
      boxExtents=<10 10 10>
      constraintSetup {angularLimitsMax=<40 45 45>, angularLimitsMin=<-30 -45 -45>, angularTarget=<0 0 0>}
      localJointOffset=<1 0 0>
     }]
   }
   n7 = Anim Dynamics - Bone: dyn_chain_01_r!(ComponentPose=@n6, bAlphaBoolEnabled:bool=True, GravityOverride:Vector="0,0,-980", ExternalForce="0,0,0") -> (Pose) {
    alphaInputType=Bool
    angularDampingOverride=0.94
    bOverrideAngularDamping=true
    bUseGravityOverride=true
    linearDampingOverride=0.9
    alphaBoolBlend {blendOption=Linear}
    boundBone {boneName=dyn_chain_01_r}
    componentAppliedLinearAccClamp=<200 200 200>
    componentLinearAccScale=<0.5 0.5 0.5>
    componentLinearVelScale=<0.9 0.9 0.9>
    physicsBodyDefinitions [{
      boundBone {boneName=dyn_chain_01_r}
      boxExtents=<10 10 10>
      constraintSetup {angularLimitsMax=<0 0 80>, angularLimitsMin=<0 0 -80>, angularTarget=<0 0 0>}
      localJointOffset=<1 0 0>
     }]
   }
   n8 = Anim Dynamics - Bone: dyn_chain_02_r!(ComponentPose=@n7, bAlphaBoolEnabled:bool=True, GravityOverride:Vector="0,0,-980", ExternalForce="0,0,0") -> (Pose) {
    alphaInputType=Bool
    angularDampingOverride=0.94
    bOverrideAngularDamping=true
    bUseGravityOverride=true
    linearDampingOverride=0.9
    alphaBoolBlend {blendOption=Linear}
    boundBone {boneName=dyn_chain_02_r}
    componentAppliedLinearAccClamp=<200 200 200>
    componentLinearAccScale=<0.5 0.5 0.5>
    componentLinearVelScale=<0.9 0.9 0.9>
    physicsBodyDefinitions [{
      boundBone {boneName=dyn_chain_02_r}
      boxExtents=<10 10 10>
      constraintSetup {angularLimitsMax=<40 30 30>, angularLimitsMin=<-30 -30 -30>, angularTarget=<0 0 0>}
      localJointOffset=<1 0 0>
     }]
   }
   n9 = Anim Dynamics - Bone: dyn_chain_03_r!(ComponentPose=@n8, bAlphaBoolEnabled:bool=True, GravityOverride:Vector="0,0,-980", ExternalForce="0,0,0") -> (Pose) {
    alphaInputType=Bool
    angularDampingOverride=0.94
    bOverrideAngularDamping=true
    bUseGravityOverride=true
    linearDampingOverride=0.9
    alphaBoolBlend {blendOption=Linear}
    boundBone {boneName=dyn_chain_03_r}
    componentAppliedLinearAccClamp=<200 200 200>
    componentLinearAccScale=<0.5 0.5 0.5>
    componentLinearVelScale=<0.9 0.9 0.9>
    physicsBodyDefinitions [{
      boundBone {boneName=dyn_chain_03_r}
      boxExtents=<10 10 10>
      constraintSetup {angularLimitsMax=<40 45 45>, angularLimitsMin=<-30 -45 -45>, angularTarget=<0 0 0>}
      localJointOffset=<1 0 0>
     }]
   }
   Output Pose(Result=Component To Local(ComponentPose=@n9))
  ]}
 EventGraph {nodes [
   Event BlueprintUpdateAnimation()
   n1 = GetOwnerSkeletalMeshComponent()
   n2 = BreakRotator(InRot=BreakTransform(InTransform=GetDeltaTransformFromRefPose(self=@n1, BoneName=spine_05, BaseName=spine_01)).Rotation)
   > Set Backpack Rotate(Backpack Rotate:Rotator=MakeRotator(Roll=MapRangeClamped(Value=@n2.Yaw, InRangeA=0, InRangeB=-30, OutRangeA=7, OutRangeB=-20))) -> (Output_Get:Rotator)
   > Set Backpack Translation(Backpack Translation:Vector=MakeVector(Y=MapRangeClamped(Value=@n2.Yaw, InRangeA=0, InRangeB=-30, OutRangeB=-4))) -> (Output_Get:Vector)
   > n3 = Branch(Condition=Get bIsSkyDiving())
   > n4 = Branch(Condition=Get bIsParachuteOpened())
   > Set GravityOverride(GravityOverride="0,0,-100")
   Set GravityOverride(execute=@n4.else, GravityOverride="0,0,400")
   Set GravityOverride(execute=@n3.else, GravityOverride="0,0,-980")
  ]}
]
variables [
 Backpack Rotate {type=Rotator}
 Backpack Translation {type=Vector}
 GravityOverride {type=Vector}
]
